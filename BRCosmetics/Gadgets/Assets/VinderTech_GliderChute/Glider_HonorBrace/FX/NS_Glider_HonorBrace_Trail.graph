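# Niagara system for the Honor Brace glider's trail effects, with variants for gameplay and front-end preview.
# GliderTrail rebuilds socket-attached ribbons each frame from ten persistent particles, offsetting them along socket rotation and smoothed system motion.
# ZGlow adds a socket-attached sprite, light and mesh flash, while Zaps scatters animated-material meshes across the socket.
# System age and timeline drive deploy flare and intensity, including speedline sprites steered by ForwardDot and RightDot.
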NiagaraSystem /BRCosmetics/Gadgets/Assets/VinderTech_GliderChute/Glider_HonorBrace/FX/NS_Glider_HonorBrace_Trail
dependencies {
 usedDataInterfaces [
  &NiagaraDataInterfaceCurve
  &NiagaraDataInterfaceColorCurve
  &NiagaraDataInterfaceVector2DCurve
  &NiagaraDataInterfaceSkeletalMesh
  &NiagaraDataInterfaceSocketReader
  &NiagaraDataInterfaceArrayFloat
 ]
 usedDynamicInputs [
  &MakeQuaternion
  &ModuloFloat
  &RemapRange
  &SelectFloatFromArray
  &MakeIntFromFloat
  &NormalizeVector
  &TransformVector
 ]
 usedModules [
  &self:SystemSpawnScript.SystemScriptSource.SystemScriptGraph.NiagaraNodeAssignment_0.SetVariables_#1
  &self:SystemSpawnScript.SystemScriptSource.SystemScriptGraph.NiagaraNodeAssignment_1.SetVariables_#2
  &SystemState
  &self:SystemSpawnScript.SystemScriptSource.SystemScriptGraph.NiagaraNodeAssignment_2.SetVariables_#3
  &Timeline
  &FN_Integrator
  &InterpolateOverTime
 ]
}
emitters {
 NE_GliderTrail {
  emitterSpawnScript [/SetVariables() {isSetParameters=true}]
  emitterUpdateScript [
   &EmitterState()
   /SetVariables(Emitter.CustomTimer:NiagaraFloat=&Add_Float(A=$Emitter.CustomTimer, B=&Lerp_Float(A=-0.03, B=-0.06, Alpha=$System.Timeline.NormalizedTime))) {isSetParameters=true}
   &SpawnBurst_Instantaneous(Spawn Count=10, Use Spawn Probability=false, UseLoopCountLimit=true)
  ]
  particleSpawnScript [
   &PartitionParticles(Number Of Partitions=1, Write Partition to Ribbon ID=::NewEnumerator0)
   &InitializeParticle(
    Color=<rgba 1 1 1 1.5>
    Lifetime=0.2
    Position Offset=<0 0 25>
    Ribbon Width=&FN_FrontEndFloatModifier(Gameplay=65, Front End=40)
    Ribbon Width Mode=::NewEnumerator1
    Sprite Size Mode=::NewEnumerator3
    Uniform Sprite Size=5
    UsePositionOffset=false
    Write Material Random=true
   )
   /SetVariables(
    Particles.RibbonID=/Niagara/DynamicInputs/Meshes/MakeRibbonIDFromSpawnGroup()
    Particles.RibbonLinkOrder:NiagaraFloat=&ReturnNormalizedExecIndex()
    Particles.SpawnGroup=$Particles.PartitionParticles.Partition
   ) {isSetParameters=true}
   /SetVariables(Particles.BackSmokeRibbonUOvverride:NiagaraFloat=&ReturnNormalizedExecIndex()) {isSetParameters=true}
  ]
  particleUpdateScript [
   /SetVariables(Particles.FadeIn:NiagaraFloat=&FloatFromCurve(
     FloatCurve=DataInterface() {propertyValues {Curve {keys [
         {arriveTangent=2.099, interpMode=RCIM_Cubic, leaveTangent=2.099, tangentMode=RCTM_User, time=0, value=0}
         {arriveTangent=-0.1461, interpMode=RCIM_Cubic, leaveTangent=-0.1461, tangentMode=RCTM_User, time=0.05, value=1}
        ]}}}
     CurveIndex=&Multiply_Float(A=$System.Age, B=1)
    )) {isSetParameters=true}
   /SetVariables(Particles.DeployAlpha:NiagaraFloat=&FloatFromCurve(
     FloatCurve=DataInterface() {propertyValues {Curve {keys [
         {arriveTangent=8.1927, interpMode=RCIM_Cubic, leaveTangent=8.1927, tangentMode=RCTM_User, time=0, value=0}
         {interpMode=RCIM_Cubic, tangentMode=RCTM_User, time=0.05, value=1}
         {arriveTangent=-0.489, interpMode=RCIM_Cubic, leaveTangent=-0.489, tangentMode=RCTM_User, time=0.5, value=0}
        ]}}}
     CurveIndex=&Multiply_Float(A=$System.Age, B=1)
    )) {isSetParameters=true}
   &ParticleState(Kill Particles When Lifetime Has Elapsed=false, Let Infinitely Lived Particles Die When Emitter Deactivates=true)
   &SocketLocation(
    SamplePosOffset=true
    Sampled Position Offset=<0 10 0>
    Socket Reader=$System.Socket Reader
    Socket Sampling Mode=::NewEnumerator3
    VelocitySpeedLimit=false
   )
   /SetVariables(Particles.Position=&LerpPosition(
     A=$Particles.Position
     B=&AddVectorToPosition(
      Position=$Particles.Position
      Vector=&Add_Vector(
       A=&MakeVector(
        X=&FloatFromCurve(
         FloatCurve=DataInterface() {propertyValues {Curve {keys [<at 0 0 RCIM_Cubic>, {arriveTangent=2.1213, interpMode=RCIM_Cubic, leaveTangent=2.1213, tangentMode=RCTM_User, time=1, value=1}]}}}
         CurveIndex=&ReturnNormalizedExecIndex()
         Scale Curve=&Multiply_Float(
          A=&MakeFloatFromVector(Vector=$System.InterpolateOverTime.InterpolateOverTime_V3_.Moving Average, Channel=::NewEnumerator0)
          B=&Lerp_Float(
           A=0
           B=&Lerp_Float(A=-35, B=-100, Alpha=$System.Timeline.NormalizedTime)
           Alpha=$System.InterpolateOverTime001.InterpolateOverTime_Float.Moving Average
          )
         )
        )
        Y=0
        Z=0
       )
       B=&Multiply_VectorByFloat(
        Vector=&RotateVectorByQuaternion(Vector To Rotate=<0 1 0>, Quaternion=$Output.SocketLocation.SampledRotation)
        Float=&FloatFromCurve(
         FloatCurve=DataInterface() {propertyValues {Curve {keys [{tangentMode=RCTM_User, time=0, value=0}, {tangentMode=RCTM_User, time=1, value=1}]}}}
         CurveIndex=&ReturnNormalizedExecIndex()
         Scale Curve=&Lerp_Float(A=&Lerp_Float(A=200, B=300, Alpha=$System.Timeline.NormalizedTime), B=400, Alpha=$Particles.DeployAlpha)
        )
       )
      )
     )
     Alpha=&ReturnNormalizedExecIndex()
    )) {isSetParameters=true}
   &ScaleColor(
    Curve Index=$Particles.RibbonLinkOrder
    Linear Color Curve=DataInterface() {propertyValues {
      AlphaCurve {keys [<at 0.0886 1>,<at 0.9269 1>,<at 1 0>]}
      BlueCurve {keys [<at 0.0213 1>]}
      GreenCurve {keys [<at 0.0213 1>]}
      RedCurve {keys [<at 0.0213 1>]}
     }}
    Scale Mode=::NewEnumerator2
   )
   &ScaleColor(
    Scale Alpha=2
    Scale RGB=&MakeVectorFromLinearColorRGB(Color=&ColorFromCurve(
      ColorCurve=DataInterface() {propertyValues {
        AlphaCurve {keys [<at 1 1 RCIM_Cubic>]}
        BlueCurve {keys [<at 0 2 RCIM_Cubic>,<at 1 5 RCIM_Cubic>]}
        GreenCurve {keys [<at 0 2 RCIM_Cubic>,<at 1 5 RCIM_Cubic>]}
        RedCurve {keys [<at 0 2 RCIM_Cubic>,<at 1 5 RCIM_Cubic>]}
       }}
      CurveIndex=&Max_Float(A:NiagaraFloat=$Particles.DeployAlpha, B:NiagaraFloat=$System.Timeline.NormalizedTime)
     ))
    ScaleA=true
    ScaleRGB=true
   ) {module=ScaleColor001}
   &ScaleRibbonWidth(Ribbon Width Scale=&Multiply_Float(
     A=&FloatFromCurve(
      FloatCurve=DataInterface() {propertyValues {Curve {keys [{tangentMode=RCTM_User, time=0.5, value=1}, {tangentMode=RCTM_User, time=1, value=0.5}]}}}
      CurveIndex=$Particles.Initial.RibbonLinkOrder
     )
     B=&Multiply_Float(
      A=&Lerp_Float(A=1, B=1.1, Alpha=$System.Timeline.NormalizedTime)
      B=&Lerp_Float(A=1, B=0.5, Alpha=$Particles.DeployAlpha)
     )
    ))
   /SetVariables(Particles.RibbonU0Override=&OneMinusFloat(Float=$Particles.RibbonLinkOrder)) {isSetParameters=true}
   &DynamicMaterialParameters(
    Index 0 Param 1=$Emitter.CustomTimer
    Index 0 Param 2=&Max_Float(A:NiagaraFloat=$System.Timeline.NormalizedTime, B:NiagaraFloat=$Particles.DeployAlpha)
    Module.Index0Float0Write=true
    Module.Index0Float1Write=true
    Module.Index0Float2Write=true
    Module.Index0Float3Write=true
    Module.Index1Float0Write=true
    Module.Index1Float1Write=true
    Module.Index1Float2Write=true
    Module.Index1Float3Write=true
    Module.Index3Float0Write=true
    Module.Index3Float1Write=true
    Module.Index3Float2Write=true
    Module.Index3Float3Write=true
    Write Parameter Index 1=::NewEnumerator1
    Write Parameter Index 3=::NewEnumerator1
   )
   /SetVariables(Particles.RibbonFacing=&RotateVectorByQuaternion(Vector To Rotate=<0 0 1>, Quaternion=$Output.SocketLocation.SampledRotation)) {isSetParameters=true}
   /SetVariables(Particles.GlowTrailWidth:NiagaraFloat=&Multiply_Float(A=$Particles.RibbonWidth, B=1.5)) {isSetParameters=true}
   /SetVariables(Particles.GlowTrailColor:LinearColor=&MultiplyLinearColor(Linear Color=$Particles.Color, Scale Factor=&MakeVector4_FromLinearColor(LinearColor=<rgba 1 1 1 0.2>))) {isSetParameters=true}
  ]
  rendererClasses [&NiagaraRibbonRendererProperties]
  renderers [
   {rendererClass=&NiagaraRibbonRendererProperties}
   {rendererClass=&NiagaraRibbonRendererProperties, rendererIndex=1}
  ]
 }
 NE_Speedlines {
  emitterUpdateScript [
   &EmitterState()
   &SpawnRate(
    Spawn Probability=&MakeCustomFloatFromBool(Bool=&SetBoolByFloatComparison(A=$System.Timeline.NormalizedTime, B=0), True Float=1, False Float=0)
    SpawnRate=&MakeCustomFloatFromBool(Bool=$User.bIsFrontEndPreview, True Float=50, False Float=90)
    Use Spawn Probability=true
   ) {module=SpawnRate001}
  ]
  particleSpawnScript [
   &InitializeParticle(
    Color=&MakeLinearColorFromVectorAndFloat(
     "Vector (RGB)"=&MakeVectorFromLinearColorRGB(Color=&ColorFromCurve(
       ColorCurve=DataInterface() {propertyValues {
         AlphaCurve {keys [<at 0 1 RCIM_Cubic>]}
         BlueCurve {keys [<at 0.4056 0 RCIM_Cubic>,<at 0.4321 0 RCIM_Cubic>]}
         GreenCurve {keys [<at 0.4056 0.7956 RCIM_Cubic>,<at 0.4321 0 RCIM_Cubic>]}
         RedCurve {keys [<at 0.4056 1 RCIM_Cubic>,<at 0.4321 1 RCIM_Cubic>]}
        }}
       CurveIndex=&RandomRangeFloat(Minimum=0, Maximum=1)
      ))
     "Float (Alpha)"=&Multiply_Float(
      A=1
      B=&MakeCustomFloatFromBool(
       Bool=$User.bIsFrontEndPreview
       True Float=&FloatFromCurve(
        FloatCurve=DataInterface() {propertyValues {Curve {keys [{tangentMode=RCTM_User, time=0, value=0}, {tangentMode=RCTM_User, time=2, value=1}]}}}
        CurveIndex=$Emitter.Age
        Scale Curve=0.4
       )
       False Float=1
      )
     )
    )
    Lifetime Max=0.2
    Lifetime Min=0.15
    Lifetime Mode=::NewEnumerator1
    Position Offset=&FN_FrontEndVector3Modifier(Gameplay=<0 -50 0>, Front End=<0 50 0>)
    Sprite Size Max=<3 10>
    Sprite Size Min=<2 2>
    Sprite Size Mode=::NewEnumerator2
    UsePositionOffset=true
    Write Material Random=true
   )
   &AddVelocity(
    Module.Use Falloff Distance=false
    Override Seed=false
    Velocity Mode=::NewEnumerator1
    Velocity Speed=&FN_FrontEndFloatModifier(Gameplay=-2500, Front End=2500)
   ) {module=AddVelocity001}
   &TorusLocation(
    Align Torus With Axis=true
    Coordinate Space:ENiagaraCoordinateSpace=::Simulation
    Handle Radius=0
    Large Radius:NiagaraFloat=&FN_FrontEndFloatModifier(
     Gameplay=&RandomRangeFloat(Minimum=350, Maximum=450)
     Front End=&RandomRangeFloat(Minimum=250, Maximum=350)
    )
    MaskSpawnGroup=false
    Module.Override Randomness=false
    Offset=<0 33 0>
    Torus Axis=<0 1 0>
    Torus Origin:NiagaraPosition=&AddVectorToPosition(
     Position=$Particles.Position
     Vector=&FN_FrontEndVector3Modifier(Gameplay=<0 -400 0>, Front End=<0 -800 0>)
    )
   )
   &AddVelocity(
    Velocity=&Add_Vector(
     A=&RandomRangeVector(Minimum=<0 300 0>, Maximum=<0 700 1>)
     B=&Add_Vector(
      A=&Multiply_VectorByFloat(Vector=<-555 0 0>, Float=$User.RightDot)
      B=&Multiply_VectorByFloat(Vector=<0 0 222>, Float=$User.ForwardDot)
     )
    )
    Velocity Speed Scale=&FN_FrontEndFloatModifier(Gameplay=1, Front End=1)
   )
  ]
  particleUpdateScript [
   &SolveForcesAndVelocity(Clamp Velocity=false, Limit Acceleration=false)
   &ParticleState()
   &ScaleSpriteSize(Scale Factor=&Vector2DFromCurve(
     Vector2Curve=DataInterface() {propertyValues {XCurve {keys [<at 0 0.5 RCIM_Cubic>,<at 0.5 1 RCIM_Cubic>]}, YCurve {keys [<at 0 0 RCIM_Cubic>,<at 0.5 1 RCIM_Cubic>]}}}
     ScaleCurve=&Vector2DFromFloat(Value=1)
    ))
   &ScaleSpriteSizeBySpeed(Max Scale Factor=<1 50>, Min Scale Factor=<0.01 0.05>, Velocity Threshold=1)
   &ScaleColor(
    Scale Alpha=&FloatFromCurve(
     FloatCurve=DataInterface() {propertyValues {Curve {keys [
         <at 0 0 RCIM_Cubic>
         <at 0.3759 1.0207>
         {arriveTangent=-2.221, interpMode=RCIM_Cubic, leaveTangent=-2.221, tangentMode=RCTM_User, time=1, value=0}
        ]}}}
     Scale Curve=&MakeCustomFloatFromBool(
      Bool=$User.bIsFrontEndPreview
      True Float=0.5
      False Float=&Lerp_Float(
       A=0
       B=0.5
       Alpha=&FN_FrontEndFloatModifier(Gameplay=$System.Timeline.NormalizedTime, Front End=0.5)
      )
     )
    )
    Scale RGB=&VectorFromFloat(Value=&FN_FrontEndFloatModifier(Gameplay=0.5, Front End=0.33))
    ScaleA=true
    ScaleRGB=true
   )
  ]
  rendererClasses [&NiagaraSpriteRendererProperties]
  renderers [{rendererClass=&NiagaraSpriteRendererProperties}]
 }
 NE_ZGlow {
  emitterUpdateScript [
   &EmitterState(Life Cycle Mode=::NewEnumerator1, Loop Behavior=::NewEnumerator1, Loop Duration=1, UseLoopDelay=false)
   &SpawnBurst_Instantaneous(Use Spawn Probability=false, UseLoopCountLimit=false)
  ]
  particleSpawnScript [
   &InitializeParticle(
    Color Mode=::NewEnumerator0
    Lifetime=0.3
    Mesh Scale Mode=::NewEnumerator3
    Sprite Size Mode=::NewEnumerator3
    Uniform Sprite Size=250
    UsePositionOffset=true
    Write Material Random=true
   )
   &InitialMeshOrientation(Orientation Axis=<0 0 1>, Orientation Vector=<0 0 1>, Use Orientation Vector=true, Use Rotation Vector=true)
   &CameraOffset(Camera Offset Amount=-50)
   /SetVariables(Particles.MeshCamOffset=10) {isSetParameters=true}
  ]
  particleUpdateScript [
   &ParticleState()
   &SocketLocation(SamplePosOffset=true, Socket Reader=$System.Socket Reader, Socket Sampling Mode=::NewEnumerator3, VelocitySpeedLimit=false)
   &ScaleSpriteSize(Uniform Curve Sprite Scale=DataInterface() {propertyValues {Curve {keys [
        {arriveTangent=2.1756, interpMode=RCIM_Cubic, leaveTangent=2.1756, tangentMode=RCTM_User, time=0, value=0}
        {interpMode=RCIM_Cubic, tangentMode=RCTM_User, time=1, value=1}
       ]}}})
   &Color(Color=&MultiplyLinearColor(
     Linear Color=&ColorFromCurve(ColorCurve=DataInterface() {propertyValues {
        AlphaCurve {keys [<at 0.5878 1 RCIM_Cubic>,<at 0.9954 0>]}
        BlueCurve {keys [<at 0 1.1 RCIM_Cubic>,<at 0.2474 0>,<at 0.7483 0 RCIM_Cubic>]}
        GreenCurve {keys [<at 0 2.5703 RCIM_Cubic>,<at 0.2474 1.5761>,<at 0.7483 0 RCIM_Cubic>]}
        RedCurve {keys [<at 0 5 RCIM_Cubic>,<at 0.2474 5>,<at 0.7483 1 RCIM_Cubic>]}
       }})
     Scale Factor=&MakeVector4FromVectorAndFloat("Vector (RGB)"=<1 1 1>, "Float (Alpha)"=1)
    ))
   &FN_Sprite_Attributes(
    Color=&MultiplyLinearColor(Linear Color=$Particles.Color, Scale Factor=&MakeVector4_FromLinearColor(LinearColor=<rgba 1 1 1 0.1>))
    Create Color Attribute=true
    Create SpriteSize Attribute=true
    SpriteSize=&Multiply_Vector2D_ByFloat(Float=2, Vector2D=$Particles.SpriteSize)
    Write to Module Namespace=true
   )
   &FN_Light_Attributes(
    Color=$Particles.Color
    Create Color Attribute=true
    Create Exponent Attribute=true
    Create Radius Attribute=true
    Enable Particle Light Enable=true
    Exponent=2.5
    Light Enabled=true
    Radius=175
    Scale Color=&Multiply_Float(
     A=20
     B=&FloatFromCurve(FloatCurve=DataInterface() {propertyValues {Curve {keys [
          {arriveTangent=4.8174, interpMode=RCIM_Cubic, leaveTangent=4.8174, tangentMode=RCTM_User, time=0, value=0}
          {interpMode=RCIM_Cubic, tangentMode=RCTM_User, time=0.15, value=1}
          {interpMode=RCIM_Cubic, tangentMode=RCTM_User, time=1, value=0}
         ]}}})
    )
   )
   &DynamicMaterialParameters(
    Index 3 Param 1=1
    Index 3 Param 4=&Multiply_Float(A=$Engine.System.Age, B=-2)
    Module.Index3Float0Write=true
    Module.Index3Float1Write=true
    Module.Index3Float2Write=true
    Module.Index3Float3Write=true
    Write Parameter Index 0=::NewEnumerator0
    Write Parameter Index 3=::NewEnumerator1
   )
   &FN_Mesh_Attributes(
    Color=&ColorFromCurve(ColorCurve=DataInterface() {propertyValues {
       AlphaCurve {keys [<at 0.5414 1 RCIM_Cubic>,<at 1 0 RCIM_Cubic>]}
       BlueCurve {keys [<at 0.18 10 RCIM_Cubic>,<at 0.3787 0>,<at 1 1 RCIM_Cubic>]}
       GreenCurve {keys [<at 0.18 10 RCIM_Cubic>,<at 0.3787 0>,<at 1 1 RCIM_Cubic>]}
       RedCurve {keys [<at 0.18 10 RCIM_Cubic>,<at 0.3787 9.2682>,<at 1 1 RCIM_Cubic>]}
      }})
    Create Color Attribute=true
    Create Scale Attribute=true
    Scale=&Lerp_Vector(
     A=<1 1 1>
     B=&FN_FrontEndVector3Modifier(Gameplay=<1.5 1 7>, Front End=<1.5 1 3>)
     Alpha=&FloatFromCurve(FloatCurve=DataInterface() {propertyValues {Curve {keys [
          {arriveTangent=14.9365, interpMode=RCIM_Cubic, leaveTangent=14.9365, tangentMode=RCTM_User, time=0, value=0}
          {arriveTangent=-2.3901, interpMode=RCIM_Cubic, leaveTangent=-2.3901, tangentMode=RCTM_User, time=0.1, value=1}
          {arriveTangent=-0.2393, interpMode=RCIM_Cubic, leaveTangent=-0.2393, tangentMode=RCTM_User, time=1, value=0}
         ]}}})
    )
    Write to Module Namespace=true
   )
  ]
  rendererClasses [
   &NiagaraSpriteRendererProperties
   &NiagaraLightRendererProperties
   &NiagaraMeshRendererProperties
  ]
  renderers [
   {rendererClass=&NiagaraSpriteRendererProperties}
   {rendererClass=&NiagaraLightRendererProperties, rendererIndex=1}
   {rendererClass=&NiagaraMeshRendererProperties, rendererIndex=2}
  ]
 }
 NE_Zaps {
  emitterUpdateScript [
   &EmitterState(
    Inactive Response=::NewEnumerator1
    Life Cycle Mode=::NewEnumerator1
    Loop Behavior=::NewEnumerator1
    Loop Duration Mode=::NewEnumerator1
    UseLoopDelay=false
   )
   &SpawnBurst_Instantaneous(Spawn Count=8, Use Spawn Probability=false, UseLoopCountLimit=false) {module=SpawnBurst_Instantaneous001}
   &SpawnRate(SpawnGroup=1, SpawnRate=&Multiply_Float(A=10, B=$System.Timeline.NormalizedTime), Use Spawn Probability=false)
  ]
  particleSpawnScript [
   /SetVariables() {isSetParameters=true}
   /SetVariables(Particles.Age:NiagaraFloat=&MaskFloatBySpawnGroup(
     Masked Float=&Lerp_Float(A=0, B=-0.5, Alpha=&ReturnNormalizedExecIndex())
     Passthrough Float=0
    )) {isSetParameters=true}
   &InitializeParticle(
    Lifetime=&MaskFloatBySpawnGroup(
     Masked Float=&RandomRangeFloat(Minimum=0.4, Maximum=0.5)
     Passthrough Float=&RandomRangeFloat(Minimum=0.4, Maximum=0.5)
    )
    Mesh Scale Mode=::NewEnumerator3
    Mesh Uniform Scale=&Multiply_Float(
     A=&MaskFloatBySpawnGroup(
      Masked Float=&RandomRangeFloat(Minimum=0.2, Maximum=0.4)
      Passthrough Float=&RandomRangeFloat(Minimum=0.15, Maximum=0.25)
     )
     B=&FN_FrontEndFloatModifier(Gameplay=1, Front End=0.8)
    )
    Sprite Size Mode=::NewEnumerator3
    Uniform Sprite Size=20
    UsePositionOffset=false
    Write Material Random=true
   )
   &InitialMeshOrientation(
    Mesh Orientation Mode=::NewEnumerator2
    Orientation Vector=&RandomVector(Vector Scale=1)
    Use Rotation Vector=true
   )
   &CameraOffset(Camera Offset Amount=10)
  ]
  particleUpdateScript [
   &ParticleState()
   &SocketLocation(
    SamplePosOffset=true
    Sampled Position Offset=&Lerp_Vector(
     A=<0 -125 2>
     B=<0 125 2>
     Alpha=&MaskFloatBySpawnGroup(
      Masked Float=&ReturnNormalizedExecIndex()
      Passthrough Float=&RandomRangeFloat(Minimum=0, Maximum=1)
     )
    )
    Socket Reader=$System.Socket Reader
    Socket Sampling Mode=::NewEnumerator3
    VelocitySpeedLimit=false
   )
   &ScaleMeshSize(Scale Factor=&Multiply_VectorByFloat(
     Vector=&RandomRangeVector(Minimum=<1 1 2.2>, Maximum=<1.3 1.3 2.2>)
     Float=&FloatFromCurve(FloatCurve=DataInterface() {propertyValues {Curve {keys [<at 0 1>,<at 1 1.5>]}}})
    )) {module=ScaleMeshSize001}
   &DynamicMaterialParameters(
    Index 0 Param 2=&RandomRangeFloat(Minimum=0, Maximum=0)
    Index 0 Param 4=13
    Index 1 Param 1=&FloatFromCurve(FloatCurve=DataInterface() {propertyValues {Curve {keys [<at 0 0.5>,<at 1 1>]}}})
    Index 1 Param 2=2
    Index 1 Param 3=&FloatFromCurve(FloatCurve=DataInterface() {propertyValues {Curve {keys [{interpMode=RCIM_Cubic, tangentMode=RCTM_Break, time=0.5, value=-0.5}, <at 1 0.1>]}}})
    Index 2 Param 1=&Multiply_Float(
     A=&FloatFromCurve(FloatCurve=DataInterface() {propertyValues {Curve {keys [{interpMode=RCIM_Cubic, leaveTangent=8.1085, tangentMode=RCTM_Break, time=0.8568, value=1}, {tangentMode=RCTM_User, time=1, value=0}]}}})
     B=&Multiply_Float(
      A=&Divide_Float(
       A=&MakeFloatFromVector(Vector=$Particles.Scale, Channel=::NewEnumerator0)
       B=&RandomRangeFloat(Minimum=10, Maximum=15)
      )
      B=1
     )
    )
    Index 2 Param 2=&FloatFromCurve(
     FloatCurve=DataInterface() {propertyValues {Curve {keys [
         {interpMode=RCIM_Cubic, tangentMode=RCTM_Break, time=0.7, value=0}
         {arriveTangent=-0.0859, interpMode=RCIM_Cubic, tangentMode=RCTM_Break, time=1, value=-1}
        ]}}}
     Scale Curve=50
    )
    Index 2 Param 3=&Waveform(
     Global Input Position=&RandomRangeFloat(Minimum=0, Maximum=15)
     Global Phase=$Particles.Age
     "Waveform [1]"=::NewEnumerator12
     "[1] Phase Scale"=2
     "[1] Frequency"=1
     "[1] Amplitude Min/Max"=<0 1>
    )
    Index 2 Param 4=25
    Index 3 Param 1=&FloatFromCurve(
     FloatCurve=DataInterface() {propertyValues {Curve {keys [
         {interpMode=RCIM_Cubic, leaveTangent=-0.0025, tangentMode=RCTM_Break, time=0, value=0}
         {arriveTangent=1.8528, interpMode=RCIM_Cubic, tangentMode=RCTM_Break, time=1, value=1}
        ]}}}
     Scale Curve=&RandomRangeFloat(Minimum=0, Maximum=0)
    )
    Index 3 Param 2=&FloatFromCurve(
     FloatCurve=DataInterface() {propertyValues {Curve {keys [
         {interpMode=RCIM_Cubic, leaveTangent=-0.0025, tangentMode=RCTM_Break, time=0, value=0.5}
         {arriveTangent=-0.0316, interpMode=RCIM_Cubic, tangentMode=RCTM_Break, time=1, value=-0.2}
        ]}}}
     Scale Curve=0.1
    )
    Index 3 Param 3=&FloatFromCurve(FloatCurve=DataInterface() {propertyValues {Curve {keys [
         {interpMode=RCIM_Cubic, leaveTangent=1.1034, tangentMode=RCTM_Break, time=0, value=-1}
         {arriveTangent=2.2399, tangentMode=RCTM_User, time=0.7, value=0.125}
        ]}}})
    Index 3 Param 4=&FloatFromCurve(FloatCurve=DataInterface() {propertyValues {Curve {keys [
         {interpMode=RCIM_Cubic, leaveTangent=3.0601, tangentMode=RCTM_Break, time=0.5, value=0}
         {arriveTangent=3.2053, tangentMode=RCTM_User, time=1, value=1}
        ]}}})
    Module.Index0Param0Write=true
    Module.Index0Param1Write=true
    Module.Index0Param2Write=true
    Module.Index0Param3Write=true
    Module.Index1Param0Write=true
    Module.Index1Param1Write=true
    Module.Index1Param2Write=true
    Module.Index1Param3Write=true
    Module.Index2Param0Write=true
    Module.Index2Param1Write=true
    Module.Index2Param2Write=true
    Module.Index2Param3Write=true
    Module.Index3Param0Write=true
    Module.Index3Param1Write=true
    Module.Index3Param2Write=true
    Module.Index3Param3Write=true
    Write Parameter Index 1=true
    Write Parameter Index 2=true
    Write Parameter Index 3=true
   )
   &SolveForcesAndVelocity(Clamp Velocity=false, Limit Acceleration=false)
   &Color(
    Color=&ColorFromCurve(ColorCurve=DataInterface() {propertyValues {
       AlphaCurve {keys [<at 0 1 RCIM_Cubic>,<at 0.7884 1>,<at 1 0 RCIM_Cubic>]}
       BlueCurve {keys [<at 0 0.9034>,<at 0.2922 0 RCIM_Cubic>,<at 0.6695 0>]}
       GreenCurve {keys [<at 0 2.2678>,<at 0.2922 0.7143 RCIM_Cubic>,<at 0.6695 0>]}
       RedCurve {keys [<at 0 5>,<at 0.2922 2 RCIM_Cubic>,<at 0.6695 0.5>]}
      }})
    Scale Alpha=&FloatFromCurve(FloatCurve=DataInterface() {propertyValues {Curve {keys [
         {interpMode=RCIM_Cubic, leaveTangent=2.5294, tangentMode=RCTM_Break, time=0, value=0}
         {arriveTangent=-0.2166, interpMode=RCIM_Cubic, tangentMode=RCTM_Break, time=0.4, value=1}
        ]}}})
   )
  ]
  rendererClasses [&NiagaraMeshRendererProperties]
  renderers [
   {rendererClass=&NiagaraMeshRendererProperties, rendererIndex=0}
   {rendererClass=&NiagaraMeshRendererProperties, rendererIndex=1}
  ]
 }
}
summary {
 systemName=NS_Glider_HonorBrace_Trail
 emitters [
  {emitterName=NE_GliderTrail, rendererClasses [&NiagaraRibbonRendererProperties]}
  {
   emitterName=NE_ZGlow
   rendererClasses [
    &NiagaraSpriteRendererProperties
    &NiagaraLightRendererProperties
    &NiagaraMeshRendererProperties
   ]
  }
  {emitterName=NE_Zaps, rendererClasses [&NiagaraMeshRendererProperties]}
  {emitterName=NE_Speedlines, rendererClasses [&NiagaraSpriteRendererProperties]}
 ]
 userVariables [
  User.bIsFrontEnd {type=NiagaraBool}
  User.bIsFrontEndPreview {type=NiagaraBool}
  User.bIsFullyDeployed {type=NiagaraBool}
  User.ForwardDot {type=NiagaraFloat, defaultValue=0, description=""}
  User.RightDot {type=NiagaraFloat, defaultValue=0, description=""}
 ]
}
